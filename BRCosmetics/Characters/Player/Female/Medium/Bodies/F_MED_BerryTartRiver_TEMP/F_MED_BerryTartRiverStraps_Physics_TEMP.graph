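# Temporary physics asset for the BerryTartRiver medium female cosmetic body, setting up physics for eight thigh straps, front and back on both legs.
# Coarse capsules cover pelvis, upper spine and thighs.
# Each strap chains from its thigh through a tiny base capsule with free swing and twist to two dyn capsules.
# Dyn joints have tight twist and mostly narrow swing limits, with one wider swing axis at each mid segment.
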PhysicsAsset /BRCosmetics/Characters/Player/Female/Medium/Bodies/F_MED_BerryTartRiver_TEMP/F_MED_BerryTartRiverStraps_Physics_TEMP
bodies {
 base_strapBack_01_l [{shapeType=Capsule, length=0.505, radius=0.505, shapeName=base_strapBack_01_l_capsule, rotation=<rot 90 -90 0>}]
 base_strapBack_01_r [{shapeType=Capsule, length=0.505, radius=0.505, shapeName=base_strapBack_01_r_capsule, rotation=<rot 90 -90 0>}]
 base_strapBack_02_l [{shapeType=Capsule, length=0.505, radius=0.505, shapeName=base_strapBack_02_l_capsule, rotation=<rot 90 -90 0>}]
 base_strapBack_02_r [{shapeType=Capsule, length=0.505, radius=0.505, shapeName=base_strapBack_02_r_capsule, rotation=<rot 90 -90 0>}]
 base_strapFront_01_l [{shapeType=Capsule, length=0.505, radius=0.505, shapeName=base_strapFront_01_l_capsule, rotation=<rot 90 -90 0>}]
 base_strapFront_01_r [{shapeType=Capsule, length=0.505, radius=0.505, shapeName=base_strapFront_01_r_capsule, rotation=<rot 90 -90 0>}]
 base_strapFront_02_l [{shapeType=Capsule, length=0.505, radius=0.505, shapeName=base_strapFront_02_l_capsule, rotation=<rot 90 -90 0>}]
 base_strapFront_02_r [{shapeType=Capsule, length=0.505, radius=0.505, shapeName=base_strapFront_02_r_capsule, rotation=<rot 90 -90 0>}]
 dyn_strapBack_01_l [{
   shapeType=Capsule
   length=3.0642
   radius=0.7905
   shapeName=dyn_strapBack_01_l_capsule
   center=<2.2541 0 0.7004>
   rotation=<rot 80.0001 -179.9997 -179.9998>
  }]
 dyn_strapBack_01_mid_l [{
   shapeType=Capsule
   length=2.6769
   radius=0.7722
   shapeName=dyn_strapBack_01_mid_l_capsule
   center=<2.1187 0.2936 0.0915>
   rotation=<rot 0 97.8908 -87.5499>
  }]
 dyn_strapBack_01_mid_r [{shapeType=Capsule, length=2.6769, radius=0.7722, shapeName=dyn_strapBack_01_mid_r_capsule, center=<2.12 0 0>, rotation=<rot 90 0 0>}]
 dyn_strapBack_01_r [{shapeType=Capsule, length=5, radius=0.7905, shapeName=dyn_strapBack_01_l_capsule, center=<3.253 0 0.5841>, rotation=<rot 85 180 -180>}]
 dyn_strapBack_02_l [{
   shapeType=Capsule
   length=4.0642
   radius=0.7905
   shapeName=dyn_strapBack_01_l_capsule
   center=<2.6411 0 0.7687>
   rotation=<rot 80.0001 -179.9997 -179.9998>
  }]
 dyn_strapBack_02_mid_l [{
   shapeType=Capsule
   length=2.5313
   radius=0.8721
   shapeName=dyn_strapBack_02_mid_l_capsule
   center=<2.027 0.7603 0.0796>
   rotation=<rot 0 110.5616 -87.8952>
  }]
 dyn_strapBack_02_mid_r [{shapeType=Capsule, length=2.5313, radius=0.8721, shapeName=dyn_strapBack_02_mid_r_capsule, center=<2.1708 0 0>, rotation=<rot 90 0 0>}]
 dyn_strapBack_02_r [{shapeType=Capsule, length=4.0642, radius=0.7905, shapeName=dyn_strapBack_01_l_capsule, center=<2.7767 0 0.7926>, rotation=<rot 80 -180 -180>}]
 dyn_strapFront_01_l [{shapeType=Capsule, length=4.0642, radius=0.7905, shapeName=dyn_strapBack_01_l_capsule, center=<2.7184 0 0.2984>, rotation=<rot 90 0 0>}]
 dyn_strapFront_01_mid_l [{
   shapeType=Capsule
   length=2.3886
   radius=0.7156
   shapeName=dyn_strapFront_01_mid_l_capsule
   center=<1.9479 0.1057 -0.1828>
   rotation=<rot 0 -86.8928 -84.6455>
  }]
 dyn_strapFront_01_mid_r [{
   shapeType=Capsule
   length=4.3886
   radius=0.7156
   shapeName=dyn_strapFront_01_mid_r_capsule
   center=<3.0029 -0.0469 -0.2783>
   rotation=<rot 0 -90.8946 -84.7063>
  }]
 dyn_strapFront_01_r [{
   shapeType=Capsule
   length=3.0642
   radius=0.7905
   shapeName=dyn_strapBack_01_l_capsule
   center=<2.3017 0 0.2984>
   rotation=<rot 90 -90 -90.0002>
  }]
 dyn_strapFront_02_l [{shapeType=Capsule, length=4.0642, radius=0.7905, shapeName=dyn_strapBack_01_l_capsule, center=<2.8531 0 0.2984>, rotation=<rot 90 0 0>}]
 dyn_strapFront_02_mid_l [{
   shapeType=Capsule
   length=2.2527
   radius=0.7683
   shapeName=dyn_strapFront_02_mid_l_capsule
   center=<1.9119 0.3287 -0.0768>
   rotation=<rot 0 -80.2436 -87.7326>
  }]
 dyn_strapFront_02_mid_r [{
   shapeType=Capsule
   length=4.2527
   radius=0.7683
   shapeName=dyn_strapFront_02_mid_r_capsule
   center=<2.9964 -0.389 -0.1196>
   rotation=<rot 0 -97.3975 -87.7326>
  }]
 dyn_strapFront_02_r [{shapeType=Capsule, length=3.0642, radius=0.7905, shapeName=dyn_strapBack_01_l_capsule, center=<2.3112 0 0.2984>, rotation=<rot 90 0 0>}]
 pelvis [{shapeType=Capsule, length=13.3913, radius=10.0424, center=<-7.623 -3.2125 0>}]
 spine_04 [{shapeType=Capsule, length=3.5238, radius=11.0455, center=<0 2.2887 0>}]
 spine_05 [{shapeType=Capsule, length=14.0223, radius=9.7378}]
 thigh_l [{shapeType=Capsule, length=19.6701, radius=9, center=<-20.0194 -0.463 0>, rotation=<rot 90 0 0>}]
 thigh_r [{shapeType=Capsule, length=19.4812, radius=9, center=<20.0194 0.4689 0>, rotation=<rot 90 0 0>}]
}
constraints [
 {bone1Name=spine_04, bone2Name=pelvis, swing1LimitDegrees=45, swing2LimitDegrees=45, twistLimitDegrees=45}
 {bone1Name=spine_05, bone2Name=spine_04, swing1LimitDegrees=45, swing2LimitDegrees=45, twistLimitDegrees=45}
 {bone1Name=thigh_l, bone2Name=pelvis, swing1LimitDegrees=45, swing2LimitDegrees=45, twistLimitDegrees=45}
 {bone1Name=thigh_r, bone2Name=pelvis, swing1LimitDegrees=45, swing2LimitDegrees=45, twistLimitDegrees=45}
 {
  bone1Name=base_strapBack_01_l
  bone2Name=thigh_l
  swing1LimitDegrees=45
  swing1Motion=Free
  swing2LimitDegrees=45
  swing2Motion=Free
  twistLimitDegrees=45
  twistMotion=Free
 }
 {bone1Name=dyn_strapBack_01_l, bone2Name=base_strapBack_01_l, swing1LimitDegrees=20, swing2LimitDegrees=20, twistLimitDegrees=5}
 {
  bone1Name=base_strapBack_02_l
  bone2Name=thigh_l
  swing1LimitDegrees=45
  swing1Motion=Free
  swing2LimitDegrees=45
  swing2Motion=Free
  twistLimitDegrees=45
  twistMotion=Free
 }
 {bone1Name=dyn_strapBack_02_l, bone2Name=base_strapBack_02_l, swing1LimitDegrees=20, swing2LimitDegrees=20, twistLimitDegrees=5}
 {
  bone1Name=base_strapFront_01_l
  bone2Name=thigh_l
  swing1LimitDegrees=45
  swing1Motion=Free
  swing2LimitDegrees=45
  swing2Motion=Free
  twistLimitDegrees=45
  twistMotion=Free
 }
 {bone1Name=dyn_strapFront_01_l, bone2Name=base_strapFront_01_l, swing1LimitDegrees=20, swing2LimitDegrees=20, twistLimitDegrees=5}
 {
  bone1Name=base_strapFront_02_l
  bone2Name=thigh_l
  swing1LimitDegrees=45
  swing1Motion=Free
  swing2LimitDegrees=45
  swing2Motion=Free
  twistLimitDegrees=45
  twistMotion=Free
 }
 {bone1Name=dyn_strapFront_02_l, bone2Name=base_strapFront_02_l, swing1LimitDegrees=20, swing2LimitDegrees=20, twistLimitDegrees=5}
 {
  bone1Name=base_strapBack_01_r
  bone2Name=thigh_r
  swing1LimitDegrees=45
  swing1Motion=Free
  swing2LimitDegrees=45
  swing2Motion=Free
  twistLimitDegrees=45
  twistMotion=Free
 }
 {bone1Name=dyn_strapBack_01_r, bone2Name=base_strapBack_01_r, swing1LimitDegrees=20, swing2LimitDegrees=20, twistLimitDegrees=5}
 {
  bone1Name=base_strapBack_02_r
  bone2Name=thigh_r
  swing1LimitDegrees=45
  swing1Motion=Free
  swing2LimitDegrees=45
  swing2Motion=Free
  twistLimitDegrees=45
  twistMotion=Free
 }
 {bone1Name=dyn_strapBack_02_r, bone2Name=base_strapBack_02_r, swing1LimitDegrees=20, swing2LimitDegrees=20, twistLimitDegrees=5}
 {
  bone1Name=base_strapFront_01_r
  bone2Name=thigh_r
  swing1LimitDegrees=45
  swing1Motion=Free
  swing2LimitDegrees=45
  swing2Motion=Free
  twistLimitDegrees=45
  twistMotion=Free
 }
 {bone1Name=dyn_strapFront_01_r, bone2Name=base_strapFront_01_r, swing1LimitDegrees=20, swing2LimitDegrees=20, twistLimitDegrees=5}
 {
  bone1Name=base_strapFront_02_r
  bone2Name=thigh_r
  swing1LimitDegrees=45
  swing1Motion=Free
  swing2LimitDegrees=45
  swing2Motion=Free
  twistLimitDegrees=45
  twistMotion=Free
 }
 {bone1Name=dyn_strapFront_02_r, bone2Name=base_strapFront_02_r, swing1LimitDegrees=20, swing2LimitDegrees=20, twistLimitDegrees=5}
 {bone1Name=dyn_strapBack_01_mid_l, bone2Name=dyn_strapBack_01_l, swing1LimitDegrees=45, swing2LimitDegrees=20, twistLimitDegrees=5}
 {bone1Name=dyn_strapBack_02_mid_l, bone2Name=dyn_strapBack_02_l, swing1LimitDegrees=45, swing2LimitDegrees=20, twistLimitDegrees=5}
 {bone1Name=dyn_strapFront_01_mid_l, bone2Name=dyn_strapFront_01_l, swing1LimitDegrees=45, swing2LimitDegrees=20, twistLimitDegrees=5}
 {bone1Name=dyn_strapFront_02_mid_l, bone2Name=dyn_strapFront_02_l, swing1LimitDegrees=45, swing2LimitDegrees=20, twistLimitDegrees=5}
 {bone1Name=dyn_strapBack_01_mid_r, bone2Name=dyn_strapBack_01_r, swing1LimitDegrees=45, swing2LimitDegrees=20, twistLimitDegrees=5}
 {bone1Name=dyn_strapBack_02_mid_r, bone2Name=dyn_strapBack_02_r, swing1LimitDegrees=45, swing2LimitDegrees=20, twistLimitDegrees=5}
 {bone1Name=dyn_strapFront_01_mid_r, bone2Name=dyn_strapFront_01_r, swing1LimitDegrees=45, swing2LimitDegrees=20, twistLimitDegrees=5}
 {bone1Name=dyn_strapFront_02_mid_r, bone2Name=dyn_strapFront_02_r, swing1LimitDegrees=45, swing2LimitDegrees=20, twistLimitDegrees=5}
]
